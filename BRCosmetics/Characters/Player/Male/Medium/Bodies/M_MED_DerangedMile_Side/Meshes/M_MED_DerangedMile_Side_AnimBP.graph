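# Post-process-style Anim Blueprint for the DerangedMile_Side male body cosmetic, following the parent mesh's pose and adjusting chest grenade bones.
# Copy Pose From Mesh feeds a deform Control Rig, which receives the crouching state.
# Cached poses chain that into weighted multi-target constraints binding six grenade bones to spine, shoulder and deform bones.
# Additive bone-space offsets then adjust five of them; the event graph has no logic.
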AnimBlueprint /BRCosmetics/Characters/Player/Male/Medium/Bodies/M_MED_DerangedMile_Side/Meshes/M_MED_DerangedMile_Side_AnimBP
target_skeleton=&/M_MED_DerangedMile_Side_Skeleton
graphs [
 AnimGraph {nodes [
   n1 = "Transform (Modify) Bone - Bone: MSTR_grenade_01_r"!(
    ComponentPose=Local To Component(LocalPose=Use cached pose 'Constaints'())
    bAlphaBoolEnabled:bool=True
    Translation="0,0,0"
    Rotation="0.138,20.116,-0.319"
    Scale="1,1,1"
   ) -> (Pose) {
    alphaInputType=Bool
    rotationMode=BMM_Additive
    rotationSpace=BCS_BoneSpace
    translationMode=BMM_Additive
    translationSpace=BCS_BoneSpace
    alphaBoolBlend {blendInTime=0.15, blendOption=Sinusoidal, blendOutTime=0.15}
    boneToModify {boneName=MSTR_grenade_01_r}
    rotation=<rot 0.138 20.116 -0.3191>
   }
   n2 = "Transform (Modify) Bone - Bone: MSTR_grenade_02_r"!(ComponentPose=@n1, bAlphaBoolEnabled:bool=True, Translation="-0.524,-1.255,-2.011", Rotation="-50.487,20.316,-15.760", Scale="1,1,1") -> (Pose) {
    alphaInputType=Bool
    rotationMode=BMM_Additive
    rotationSpace=BCS_BoneSpace
    translationMode=BMM_Additive
    translationSpace=BCS_BoneSpace
    alphaBoolBlend {blendInTime=0.15, blendOption=Sinusoidal, blendOutTime=0.15}
    boneToModify {boneName=MSTR_grenade_02_r}
    rotation=<rot -50.487 20.3161 -15.7601>
    translation=<-0.524 -1.255 -2.011>
   }
   n3 = "Transform (Modify) Bone - Bone: MSTR_grenade_03_r"!(ComponentPose=@n2, bAlphaBoolEnabled:bool=True, Translation="-0.501,-1.436,-0.107", Rotation="10.229,26.966,22.330", Scale="1,1,1") -> (Pose) {
    alphaInputType=Bool
    rotationMode=BMM_Additive
    rotationSpace=BCS_BoneSpace
    translationMode=BMM_Additive
    translationSpace=BCS_BoneSpace
    alphaBoolBlend {blendInTime=0.15, blendOption=Sinusoidal, blendOutTime=0.15}
    boneToModify {boneName=MSTR_grenade_03_r}
    rotation=<rot 10.229 26.966 22.33>
    translation=<-0.501 -1.436 -0.107>
   }
   n4 = "Transform (Modify) Bone - Bone: MSTR_grenade_03_l"!(ComponentPose=@n3, bAlphaBoolEnabled:bool=True, Translation="0.377,2.377,0.279", Rotation="48.858,35.019,-12.529", Scale="1,1,1") -> (Pose) {
    alphaInputType=Bool
    rotationMode=BMM_Additive
    rotationSpace=BCS_BoneSpace
    translationMode=BMM_Additive
    translationSpace=BCS_BoneSpace
    alphaBoolBlend {blendInTime=0.15, blendOption=Sinusoidal, blendOutTime=0.15}
    boneToModify {boneName=MSTR_grenade_03_l}
    rotation=<rot 48.8581 35.019 -12.529>
    translation=<0.377 2.377 0.279>
   }
   n5 = "Transform (Modify) Bone - Bone: MSTR_grenade_02_l"!(ComponentPose=@n4, bAlphaBoolEnabled:bool=True, Translation="1.772,2.147,-0.829", Rotation="72.313,7.692,-35.720", Scale="1,1,1") -> (Pose) {
    alphaInputType=Bool
    rotationMode=BMM_Additive
    rotationSpace=BCS_BoneSpace
    translationMode=BMM_Additive
    translationSpace=BCS_BoneSpace
    alphaBoolBlend {blendInTime=0.15, blendOption=Sinusoidal, blendOutTime=0.15}
    boneToModify {boneName=MSTR_grenade_02_l}
    rotation=<rot 72.313 7.6919 -35.7201>
    translation=<1.7723 2.1473 -0.8292>
   }
   Output Pose(Result=Component To Local(ComponentPose=@n5)) {layerGroup=None}
   n6 = Control Rig!(Source=Copy Pose From Mesh() {bUseAttachedParent=true}, Alpha, Crouching:bool=Get bIsCrouching()) -> (Pose) {controlRigAssetReference {blueprintRigClass=&/M_MED_DerangedMile_Side_DeformCtrlRig.M_MED_DerangedMile_Side_DeformCtrlRig_C}}
   Save cached pose 'Deform Control Rig'(Pose=@n6) {cacheName=Deform Control Rig}
   n7 = "Constraint - MSTR_grenade_01_r (PPPPPPP)"!(
    ComponentPose=Local To Component(LocalPose=Use cached pose 'Deform Control Rig'())
    Alpha=1
    ConstraintWeights_0=0.3755
    ConstraintWeights_1:float=0.1726
    ConstraintWeights_2:float=0.1601
    ConstraintWeights_3:float=0.1143
    ConstraintWeights_4:float=0.0976
    ConstraintWeights_5:float=0.043
    ConstraintWeights_6:float=0.0369
   ) -> (Pose) {
    boneToModify {boneName=MSTR_grenade_01_r}
    constraintSetup [
     {targetBone {boneName=spine_05}}
     {targetBone {boneName=clavicle_r}}
     {targetBone {boneName=upperarm_twist_01_r}}
     {targetBone {boneName=deform_upperarm_fr_r}}
     {targetBone {boneName=deform_pec_r}}
     {targetBone {boneName=spine_03}}
     {targetBone {boneName=upperarm_twist_02_r}}
    ]
    constraintWeights [1,1,1,1,1,1,1]
   }
   n8 = "Constraint - MSTR_grenade_02_r (PPPPPPP)"!(
    ComponentPose=@n7
    Alpha=1
    ConstraintWeights_0=0.2767
    ConstraintWeights_1:float=0.272
    ConstraintWeights_2:float=0.1355
    ConstraintWeights_3:float=0.0973
    ConstraintWeights_4:float=0.084
    ConstraintWeights_5:float=0.0734
    ConstraintWeights_6:float=0.061
   ) -> (Pose) {
    boneToModify {boneName=MSTR_grenade_02_r}
    constraintSetup [
     {targetBone {boneName=spine_05}}
     {targetBone {boneName=spine_04}}
     {targetBone {boneName=upperarm_twist_01_r}}
     {targetBone {boneName=clavicle_r}}
     {targetBone {boneName=spine_03}}
     {targetBone {boneName=deform_pec_r}}
     {targetBone {boneName=deform_upperarm_fr_r}}
    ]
    constraintWeights [1,1,1,1,1,1,1]
   }
   n9 = "Constraint - MSTR_grenade_03_r (PPPPPPPP)"!(
    ComponentPose=@n8
    Alpha=1
    ConstraintWeights_0=0.5641
    ConstraintWeights_1:float=0.1409
    ConstraintWeights_2:float=0.0952
    ConstraintWeights_3:float=0.0742
    ConstraintWeights_4:float=0.0413
    ConstraintWeights_5:float=0.0406
    ConstraintWeights_6:float=0.0328
    ConstraintWeights_7:float=0.0109
   ) -> (Pose) {
    boneToModify {boneName=MSTR_grenade_03_r}
    constraintSetup [
     {targetBone {boneName=spine_04}}
     {targetBone {boneName=spine_05}}
     {targetBone {boneName=upperarm_twist_01_r}}
     {targetBone {boneName=spine_03}}
     {targetBone {boneName=deform_pec_r}}
     {targetBone {boneName=clavicle_r}}
     {targetBone {boneName=spine_02}}
     {targetBone {boneName=deform_upperarm_fr_r}}
    ]
    constraintWeights [1,1,1,1,1,1,1,1]
   }
   n10 = "Constraint - MSTR_grenade_01_l (PPPPPPPP)"!(
    ComponentPose=@n9
    Alpha=1
    ConstraintWeights_0=0.408
    ConstraintWeights_1:float=0.134
    ConstraintWeights_2:float=0.1203
    ConstraintWeights_3:float=0.1149
    ConstraintWeights_4:float=0.0786
    ConstraintWeights_5:float=0.0749
    ConstraintWeights_6:float=0.0384
    ConstraintWeights_7:float=0.0309
   ) -> (Pose) {
    boneToModify {boneName=MSTR_grenade_01_l}
    constraintSetup [
     {targetBone {boneName=spine_05}}
     {targetBone {boneName=clavicle_l}}
     {targetBone {boneName=upperarm_twist_01_l}}
     {targetBone {boneName=deform_upperarm_fr_l}}
     {targetBone {boneName=deform_pec_l}}
     {targetBone {boneName=spine_04}}
     {targetBone {boneName=spine_03}}
     {targetBone {boneName=upperarm_twist_02_l}}
    ]
    constraintWeights [1,1,1,1,1,1,1,1]
   }
   n11 = "Constraint - MSTR_grenade_02_l (PPPP)"!(
    ComponentPose=@n10
    Alpha=1
    ConstraintWeights_0=0.417
    ConstraintWeights_1:float=0.3486
    ConstraintWeights_2:float=0.2337
    ConstraintWeights_3:float=0.0007
   ) -> (Pose) {
    boneToModify {boneName=MSTR_grenade_02_l}
    constraintSetup [
     {targetBone {boneName=spine_05}}
     {targetBone {boneName=spine_04}}
     {targetBone {boneName=upperarm_twist_01_l}}
     {targetBone {boneName=MSTR_grenade_02_l}}
    ]
    constraintWeights [1,1,1,1]
   }
   n12 = "Constraint - MSTR_grenade_03_l (PPPPPPPP)"!(
    ComponentPose=@n11
    Alpha=1
    ConstraintWeights_0=0.5692
    ConstraintWeights_1:float=0.1395
    ConstraintWeights_2:float=0.0957
    ConstraintWeights_3:float=0.0726
    ConstraintWeights_4:float=0.0404
    ConstraintWeights_5:float=0.0398
    ConstraintWeights_6:float=0.0321
    ConstraintWeights_7:float=0.0106
   ) -> (Pose) {
    boneToModify {boneName=MSTR_grenade_03_l}
    constraintSetup [
     {targetBone {boneName=spine_04}}
     {targetBone {boneName=spine_05}}
     {targetBone {boneName=upperarm_twist_01_l}}
     {targetBone {boneName=spine_03}}
     {targetBone {boneName=deform_pec_l}}
     {targetBone {boneName=clavicle_l}}
     {targetBone {boneName=spine_02}}
     {targetBone {boneName=deform_upperarm_fr_l}}
    ]
    constraintWeights [1,1,1,1,1,1,1,1]
   }
   Save cached pose 'Constaints'(Pose=Component To Local(ComponentPose=@n12)) {cacheName=Constaints}
  ]}
 EventGraph {nodes [Event BlueprintUpdateAnimation()]}
]
variables [
 ar_firingPosition {type=integer}
 WearingPistol {type=boolean}
]
