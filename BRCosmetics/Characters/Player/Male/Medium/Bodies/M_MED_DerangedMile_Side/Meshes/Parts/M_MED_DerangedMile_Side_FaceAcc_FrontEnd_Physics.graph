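# Physics asset for the front-end version of the DerangedMile_Side male medium character's face accessory, covering a two-bone dangling pendant.
# Capsule bodies sit on both pendant bones, the second using two, while large capsules on spine_04 and spine_05 approximate the upper torso.
# Constraints chain the pendant from spine_05 through dyn_pendant_01 to dyn_pendant_02 with narrow swing limits and zero-degree twist limits, while the spine-to-spine joint is free.
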PhysicsAsset /BRCosmetics/Characters/Player/Male/Medium/Bodies/M_MED_DerangedMile_Side/Meshes/Parts/M_MED_DerangedMile_Side_FaceAcc_FrontEnd_Physics
bodies {
 dyn_pendant_01 [{shapeType=Capsule, length=7.873, radius=1.3395, shapeName=dyn_pendant_01_capsule, center=<0 0 0.4787>, rotation=<rot 0 0 90>}]
 dyn_pendant_02 [
  {
   shapeType=Capsule
   length=1.7804
   radius=1.7541
   shapeName=dyn_pendant_02_capsule
   center=<3.7684 -0.0554 0.8781>
   rotation=<rot -4.5751 176.5663 91.274>
  }
  {shapeType=Capsule, length=3.2336, radius=1.5151, shapeName=dyn_pendant_02_capsule, center=<1.6323 0.0461 0.5908>, rotation=<rot 0 0 -85.539>}
 ]
 spine_04 [{shapeType=Capsule, length=7.2626, radius=13, shapeName=spine_04_capsule, center=<5.4675 -0.447 0>, rotation=<rot 0 -90 0>}]
 spine_05 [{shapeType=Capsule, length=13, radius=10, shapeName=spine_05_capsule, center=<-1.285 1.2801 0>}]
}
constraints [
 {
  bone1Name=spine_05
  bone2Name=spine_04
  swing1LimitDegrees=45
  swing1Motion=Free
  swing2LimitDegrees=45
  swing2Motion=Free
  twistLimitDegrees=45
  twistMotion=Free
 }
 {bone1Name=dyn_pendant_01, bone2Name=spine_05, swing1LimitDegrees=25, swing2LimitDegrees=35, twistLimitDegrees=0}
 {bone1Name=dyn_pendant_02, bone2Name=dyn_pendant_01, swing1LimitDegrees=15, swing2LimitDegrees=25, twistLimitDegrees=0}
]
