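# Animation Blueprint for the DerangedMile_Side character's face accessory, adding physics motion on top of the pose copied from its attached parent mesh.
# Two RigidBody simulations relative to spine_05 use separate gameplay and front-end physics assets, selected by bIsInFrontEnd.
# A front-end idle curve, Frontend_Idle_Bypass, weights a per-bone blend of DefaultSlot over the dyn_gun bones.
# Backbling flags and body mesh are stored but unused by the AnimGraph.
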AnimBlueprint /BRCosmetics/Characters/Player/Male/Medium/Bodies/M_MED_DerangedMile_Side/Meshes/Parts/M_MED_DerangedMile_Side_FaceAcc_AnimBP
target_skeleton=&/M_MED_DerangedMile_Side_FaceAcc_Skeleton
graphs [
 AnimGraph {nodes [
   n1 = Layered blend per bone(
    BasePose=Use cached pose 'StartPose'()
    BlendPoses_0=Slot 'DefaultSlot'()
    BlendWeights_0=Get Custom Idle Is Playing() -> (Custom Idle Is Playing:float)
   ) {layerSetup [{branchFilters [{boneName=dyn_gun_01}, {boneName=dyn_gun_02}, {boneName=dyn_gun_03}]}]}
   Output Pose(Result=@n1) {layerGroup=None}
   n2 = Copy Pose From Mesh() {bUseAttachedParent=true}
   Save cached pose 'Turnet'(Pose=Component To Local(ComponentPose=Local To Component(LocalPose=@n2))) {cacheName=Turnet}
   n3 = RigidBody(ComponentPose=Local To Component(LocalPose=Use cached pose 'Turnet'())) {
    bOverrideWorldGravity=true
    overridePhysicsAsset=&/M_MED_DerangedMile_Side_FaceAcc_Physics
    simulationSpace=BaseBoneSpace
    baseBoneRef {boneName=spine_05}
    componentAppliedLinearAccClamp=<2500 2500 2500>
    overrideWorldGravity {z=-980}
    simSpaceSettings {dampingAlpha=0, maxAngularAcceleration=250, maxAngularVelocity=250, maxLinearAcceleration=450, maxLinearVelocity=250, worldAlpha=1}
   }
   n4 = RigidBody(ComponentPose=Local To Component(LocalPose=Use cached pose 'Turnet'())) {
    bOverrideWorldGravity=true
    overridePhysicsAsset=&/M_MED_DerangedMile_Side_FaceAcc_FrontEnd_Physics
    simulationSpace=BaseBoneSpace
    baseBoneRef {boneName=spine_05}
    componentAppliedLinearAccClamp=<2500 2500 2500>
    overrideWorldGravity {z=-980}
    simSpaceSettings {dampingAlpha=0, maxAngularAcceleration=250, maxAngularVelocity=250, maxLinearAcceleration=450, maxLinearVelocity=250, worldAlpha=1}
   }
   Save cached pose 'StartPose'(Pose=Blend Poses by bool(BlendPose_0=Component To Local(ComponentPose=@n4), BlendPose_1=Component To Local(ComponentPose=@n3), bActiveValue=Get bIsInFrontEnd()) {blendType=Linear}) {cacheName=StartPose}
  ]}
 EventGraph {nodes [
   Event BlueprintInitializeAnimation()
   > n1 = Cast To PlayerPawn_Athena(Object=TryGetPawnOwner())
   > Set Body_Mesh(Body_Mesh=GetSkeletalMeshForPartType(self=@n1))
   Event BlueprintUpdateAnimation()
   > n2 = Sequence!(execute) -> (then_0, then_1, then_2:Exec)
   Branch(execute=@n2.then_2, Condition=Get bIsInFrontEnd())
   > Set Custom Idle Is Playing(Custom Idle Is Playing:float=GetCurveValue(self=GetOwnerFrontEndAnimInstance(), CurveName=Frontend_Idle_Bypass)) -> (Output_Get:float)
   Event OnCharacterCustomizationLoaded()
   n3 = CheckForBackbling()
   > Set isWearingPack(isWearingPack:bool=@n3) -> (Output_Get:bool)
   > Set isWearingCape(isWearingCape=AND Boolean(A=@n3, B=@n3.bIsBackblingLong))
  ]}
]
variables [
 Body_Mesh {type=SkeletalMeshComponent}
 Custom Idle Is Playing {type=number}
 isWearingCape {type=boolean}
 isWearingPack {type=boolean}
]
